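# Additive unlit Niagara sprite material for the energy effect on the RoyalDusk back bling.
# Two panning plasma texture samples, scaled by a parameter and offset per particle, form animated noise, with the first distorting and multiplying the second.
# Particle alpha and a soft radial mask shape the noise, with depth fade only at High and Epic quality.
# Particle color feeds a shared emissive-adjustment function.
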Material /Game/Accessories/FORT_Backpacks/Backpack_F_MED_RoyalDusk/FX/M_FX_RoyalDusk_Energy
graph {
 properties {bUsedWithNiagaraSprites=true, blendMode=BLEND_Additive, pixelDepthOffsetMode=PDOM_Legacy, shadingModel=MSM_Unlit, usedShadingModels=Unlit}
 nodes [
  n1 = ParticleColor()
  MaterialFunctionCall(Emissive=Multiply(A=@n1.RGB)) {materialFunction=&MF_FX_EmissiveAdjustments}
  n2 = ScalarParameter() {defaultValue=1, parameterName=Scale}
  n3 = TextureCoordinate() {uTiling=2}
  n4 = ParticleRandom()
  n5 = Panner(Coordinate=Add(A=Multiply(A=@n2, B=@n3), B=@n4)) {bFractionalPart=true, speedX=-0.1, speedY=0.1}
  n6 = Multiply(A=TextureCoordinate() {uTiling=1.5, vTiling=0.75}, B=@n2)
  n7 = TextureSample(UVs=Panner(Coordinate=Add(A=@n6, B=@n4)) {bFractionalPart=true, speedX=0.1, speedY=0.2}) {texture=/Game/Effects/Fort_Effects/Textures/Plasma/T_Plasma_01_M}
  n8 = Multiply(A=@n7.R) {constB=0.2}
  n9 = Multiply(A=TextureSample(UVs=Add(A=@n5, B=@n8)) {texture=&T_Plasma_01_M}.G, B=@n7.R)
  n10 = Multiply(
   A=Saturate(None=Multiply(A=@n1.A, B=@n9))
   B=SphereMask(A=TextureCoordinate(), B=Constant() {r=0.5}) {attenuationRadius=0.5, hardnessPercent=0}
  )
  n11 = DepthFade(Opacity=@n10, FadeDistance=ScalarParameter() {defaultValue=5, parameterName=DepthFade})
  QualitySwitch(Default=@n10, Low=@n10, High=@n11, Medium=@n10, Epic=@n11)
  Subtract(A=@n9, B=OneMinus(None=@n1.A))
 ]
}
