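# Unlit, additive, two-sided material for light ray effects on the Royal Dusk backpack, enabled for Niagara mesh particles.
# Particle color is hue-shifted along V by a dynamic parameter in the vertex stage, then scaled by particle alpha into an emissive adjustment function.
# A ray mask adds stretched noise scrolled by RaysTime to a panning RoyalDusk texture, fading along V.
# A quality switch applies depth fade except on Low and Medium.
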Material /Game/Accessories/FORT_Backpacks/Backpack_F_MED_RoyalDusk/FX/M_FX_RoyalDusk_Pack_Lighting
graph {
 properties {
  bUsedWithNiagaraMeshParticles=true
  blendMode=BLEND_Additive
  pixelDepthOffsetMode=PDOM_Legacy
  previewMesh=/Game/Effects/Fort_Effects/Mesh/SM_Convex_Circle
  shadingModel=MSM_Unlit
  twoSided=true
  usedShadingModels=Unlit
 }
 nodes [
  Constant3Vector() {constant=<rgba 4 0 0 0>}
  n1 = TextureCoordinate()
  n2 = ParticleColor()
  n3 = MaterialFunctionCall(
   Hue Shift Percentage=Multiply(A=ComponentMask(None=@n1) {g=true}, B=DynamicParameter() {paramNames [HueShift,RaysTime,Param3,Param4]}.HueShift)
   Texture=@n2.RGB
  ) {materialFunction=&HueShift}
  MaterialFunctionCall(Emissive=Multiply(A=VertexInterpolator(VS=@n3.Result).PS, B=@n2.A)) {materialFunction=&MF_FX_EmissiveAdjustments}
  n4 = Panner(
   Coordinate=Multiply(A=TextureCoordinate() {uTiling=0.5, vTiling=0.01})
   Time=DynamicParameter() {paramNames [HueShift,RaysTime,Param3,Param4]}.RaysTime
  ) {bFractionalPart=true, speedY=-1}
  n5 = TextureCoordinate() {vTiling=2}
  n6 = TextureSample(UVs=Panner(Coordinate=Multiply(A=@n5)) {bFractionalPart=true, speedY=-0.11}) {samplerType=SAMPLERTYPE_LinearGrayscale, texture=/Game/Accessories/FORT_Backpacks/Backpack_F_MED_RoyalDusk/Textures/T_RoyalDusk_RoyalDusk_M3}
  n7 = Power(Base=@n6.G) {constExponent=4}
  n8 = ComponentMask(None=TextureCoordinate()) {g=true}
  n9 = Multiply(
   A=Add(A=TextureSample(UVs=@n4) {texture=/Game/Global/Textures/Default/Noise/T_TileNoise_02}.G, B=Multiply(A=@n7) {constB=0.7})
   B=Power(Base=OneMinus(None=@n8)) {constExponent=2.5}
  )
  n10 = DepthFade(Opacity=@n9) {fadeDistanceDefault=10}
  QualitySwitch(Default=@n10, Low=@n9, High=@n10, Medium=@n9, Epic=@n10)
 ]
}
