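# Material function that adds an animated purple and pink glow to the RoyalDusk backpack's incoming material attributes.
# Camera-shifted local position yields fake-depth radial coordinates and a pointed distance field that drive scrolling pattern and line textures, sampled partly with UV derivatives.
# These layers are dimmed by time of day and feed base color and emissive.
# A parameter-positioned pink band and two counter-scrolling, multiplied texture samples add further emissive.
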MaterialFunction /Game/Accessories/FORT_Backpacks/Backpack_F_MED_RoyalDusk/Materials/MF_RoyalDusk_Backpack
graph {nodes [
  n1 = NamedRerouteUsage()
  n2 = NamedRerouteUsage()
  n3 = Multiply(A=Time() {bOverride_Period=true, period=50}) {constB=0.02}
  n4 = NamedRerouteUsage()
  n5 = Multiply(A=Time() {bOverride_Period=true, period=2}) {constB=0.5}
  n6 = TextureCoordinate()
  n7 = DDX(Value=@n6)
  n8 = DDY(Value=@n6)
  n9 = TextureSample!(UVs=AppendVector(A=Add(A=@n2, B=@n3), B=Subtract(A=@n4, B=@n5)), Tex, "DDX(UVs)"=@n7, "DDY(UVs)"=@n8, Apply View MipBias) {
   automaticViewMipBias=false
   mipValueMode=TMVM_Derivative
   texture=/Game/Accessories/FORT_Backpacks/Backpack_F_MED_RoyalDusk/Textures/T_RoyalDusk_RoyalDusk_M1_TGA
  }
  n10 = Multiply(A=@n9.RGB) {constB=2.4}
  n11 = Constant3Vector() {constant=<rgba 0.4018 0.0407 0.7812 1>}
  n12 = NamedRerouteUsage()
  n13 = OneMinus(None=Saturate(None=Multiply(A=@n12) {constB=2}))
  n14 = Multiply(A=@n13, B=@n13)
  n15 = Multiply(A=NamedRerouteUsage()) {constB=1.1}
  n16 = NamedRerouteUsage()
  n17 = OneMinus(None=Saturate(None=LinearInterpolate(A=@n15, B=@n16) {constAlpha=0.3}))
  n18 = Multiply(
   A=Multiply(A=LinearInterpolate(A=@n10, B=@n11, Alpha=@n14), B=ScalarParameter() {defaultValue=1, parameterName=CoreMultiply})
   B=Multiply(A=@n17, B=@n17)
  )
  n19 = NamedRerouteUsage()
  n20 = Subtract(A=NamedRerouteUsage(), B=@n5)
  n21 = ScalarParameter() {parameterName=PinkPatternScroll}
  n22 = TextureSample(UVs=AppendVector(A=Subtract(A=@n19, B=@n3), B=Subtract(A=@n20, B=@n21)), "DDX(UVs)"=@n7, "DDY(UVs)"=@n8) {
   mipValueMode=TMVM_Derivative
   samplerType=SAMPLERTYPE_LinearGrayscale
   texture=/Game/Accessories/FORT_Backpacks/Backpack_F_MED_RoyalDusk/Textures/T_RoyalDusk_RoyalDusk_M3
  }
  n23 = NamedRerouteUsage()
  n24 = OneMinus(None=Saturate(None=Multiply(A=@n23) {constB=3}))
  n25 = NamedRerouteUsage()
  n26 = Abs(None=Multiply(A=ComponentMask(None=@n25) {b=true, r=true}) {constB=0.88})
  n27 = ScalarParameter() {defaultValue=1.5, parameterName=LineCurve}
  n28 = Power(Base=@n26, Exp=Divide(B=@n27))
  n29 = Power(Base=Add(A=ComponentMask(None=@n28) {r=true}, B=ComponentMask(None=@n28) {g=true}), Exp=@n27)
  n30 = TextureSample!(
   UVs=Subtract(
    A=NamedRerouteDeclaration(None=@n29) {name=RadialPointedDistance, nodeColor=<rgba 0 1 0.9151 1>}
    B=ScalarParameter() {defaultValue=-0.1, parameterName=LineScroll}
   )
   Tex
   Apply View MipBias
  ) {
   samplerSource=SSM_Wrap_WorldGroupSettings
   samplerType=SAMPLERTYPE_LinearGrayscale
   texture=/Game/Accessories/FORT_Backpacks/Backpack_F_MED_RoyalDusk/Textures/T_RoyalDusk_RoyalDusk_M2
  }
  n31 = Multiply(A=Max(A=LinearInterpolate(Alpha=@n22.R) {constA=0.2, constB=0.8}, B=Multiply(A=@n24, B=@n24)), B=@n30.R)
  n32 = Saturate(None=Subtract(A=Multiply(A=@n31) {constB=8}) {constB=2})
  n33 = Multiply(A=@n32, B=@n32)
  n34 = LinearInterpolate(
   A=Constant3Vector() {constant=<rgba 0.0252 0.0535 0.4844 1>}
   B=Constant3Vector() {constant=<rgba 0.9164 0.1406 3 1>}
   Alpha=ScalarParameter() {parameterName=LineIntensity}
  )
  n35 = LinearInterpolate(
   A=@n34
   B=Multiply(A=@n34, B=ScalarParameter() {defaultValue=10, parameterName=OuterGlow}) {constB=2}
   Alpha=SmoothStep(Value=NamedRerouteUsage()) {constMax=0.7, constMin=0.55}
  )
  n36 = Constant3Vector() {constant=<rgba 0.0301 0.0015 0.2917 1>}
  n37 = Min(A=ComponentMask(None=@n26) {r=true}, B=ComponentMask(None=@n26) {g=true})
  n38 = OneMinus(None=Saturate(None=Multiply(A=@n37) {constB=32}))
  n39 = Multiply(A=@n38, B=@n38) {constB=2}
  n40 = Add(A=Add(A=Multiply(A=@n33, B=@n35), B=Multiply(A=@n31, B=@n36)), B=Multiply(A=@n33, B=Multiply(A=@n39) {constB=0.15}))
  n41 = Constant() {r=1}
  n42 = Constant() {r=0.6}
  n43 = Multiply(A=Multiply(
    A=Add(A=@n18, B=@n40)
    B=MaterialFunctionCall(Value_Morning=@n41, Value_Day=@n41, Value_Evening=@n41, Value_Night=@n42) {materialFunction=&Scalar_Time_of_Day}.Result
   )) {constB=0.5}
  n44 = NamedRerouteUsage()
  n45 = Multiply(A=NamedRerouteUsage()) {constB=0.5}
  n46 = ScalarParameter() {defaultValue=1.3372, parameterName=PinkScroll}
  n47 = Abs(None=Add(A=Subtract(A=@n45, B=@n46)) {constB=-0.5})
  n48 = Multiply(
   A=LinearInterpolate(A=Constant3Vector() {constant=<rgba 0.0535 0 0.2448 1>}, B=Constant3Vector() {constant=<rgba 2.6968 0 4 1>}, Alpha=Multiply(A=@n44, B=@n44))
   B=OneMinus(None=Saturate(None=@n47))
  )
  n49 = Multiply(A=NamedRerouteUsage()) {constB=1.24}
  n50 = Multiply(A=Time() {bOverride_Period=true, period=20}) {constB=0.05}
  n51 = Multiply(A=NamedRerouteUsage()) {constB=0.8}
  n52 = TextureSample(UVs=AppendVector(A=Add(A=@n49, B=@n50), B=@n51)) {texture=/Game/Accessories/FORT_Backpacks/Backpack_F_MED_RoyalDusk/Textures/T_RoyalDusk_RoyalDusk_M4}
  n53 = Multiply(A=NamedRerouteUsage()) {constB=1.47}
  n54 = Multiply(A=Time() {bOverride_Period=true, period=10}) {constB=0.1}
  n55 = TextureSample(UVs=AppendVector(A=Subtract(A=@n53, B=@n54), B=@n51)) {texture=&T_RoyalDusk_RoyalDusk_M4}
  n56 = Multiply(A=Multiply(A=Multiply(A=@n52.RGB, B=@n55.RGB), B=ScalarParameter() {parameterName=LineIntensity}) {constB=4}) {constB=1.3}
  n57 = PreSkinnedPosition()
  n58 = SetMaterialAttributes!(
   MaterialAttributes=NamedRerouteUsage()
   BaseColor=Multiply(A=Add(A=@n1, B=@n43), B=ScalarParameter() {defaultValue=1, parameterName=BaseColorMultiply})
   Specular=NamedRerouteUsage()
   Metallic=Constant() {r=0.2}
   Roughness=NamedRerouteUsage()
   EmissiveColor=Add(A=@n43, B=Add(A=@n48, B=@n56))
   Normal=NamedRerouteUsage()
   AmbientOcclusion=NamedRerouteUsage()
   WorldPositionOffset=Multiply(A=Transform(None=@n57) {transformSourceType=TRANSFORMSOURCE_Local}, B=ScalarParameter() {parameterName=WPO}) {constB=0}
  ) {attributeSetTypes [
    #1
    #2
    #3
    #4
    #5
    #6
    #7
    #8
   ]}
  FunctionOutput(None=@n58) {outputName=MaterialAttributes}
  n59 = GetMaterialAttributes(None=FunctionInput() {inputType=FunctionInput_MaterialAttributes}) {attributeGetTypes [
    #1
    #2
    #4
    #3
    #6
   ]}
  NamedRerouteDeclaration(None=@n59.MaterialAttributes) {name=MA, nodeColor=<rgba 0.1016 1 0 1>}
  NamedRerouteDeclaration(None=@n59.Normal) {name=Normal, nodeColor=<rgba 0.9219 0 1 1>}
  NamedRerouteDeclaration(None=@n22.R) {name=M3, nodeColor=<rgba 0 1 0.5547 1>}
  n60 = FunctionInput() {bUsePreviewValueAsDefault=true, inputName=LocalPosition, sortPriority=1}
  n61 = Transform(None=CameraVectorWS()) {transformSourceType=TRANSFORMSOURCE_World, transformType=TRANSFORM_Local}
  n62 = Add(
   A=NamedRerouteDeclaration(None=Multiply(A=@n60) {constB=0.08}) {name=PreSkinnedPosition, nodeColor=<rgba 1 0.1641 0 1>}
   B=Multiply(A=NamedRerouteDeclaration(None=@n61) {name=LocalCameraVector, nodeColor=<rgba 1 0 0.4219 1>}) {constB=-0.2}
  )
  n63 = ComponentMask(None=@n62) {b=true, r=true}
  NamedRerouteDeclaration(None=Distance(A=Multiply(A=@n63) {constB=1.5}, B=Constant())) {name=RadialDistance, nodeColor=<rgba 0 0.6236 1 1>}
  n64 = Arctangent2Fast(Y=ComponentMask(None=@n62) {r=true}, X=ComponentMask(None=@n62) {b=true})
  NamedRerouteDeclaration(None=Divide(A=ComponentMask(None=@n64) {r=true}) {constB=6.2832}) {name=RadialArc, nodeColor=<rgba 1 0 0.4219 1>}
  NamedRerouteDeclaration(None=@n59.BaseColor) {name=BaseColor, nodeColor=<rgba 1 0.8672 0 1>}
  NamedRerouteDeclaration(None=@n59.Specular) {name=Specular, nodeColor=<rgba 0 0.5078 1 1>}
  NamedRerouteDeclaration(None=@n59.Roughness) {name=Roughness, nodeColor=<rgba 1 0 0.8672 1>}
  NamedRerouteDeclaration(None=@n59.Metallic) {name=AmbientOcclusion, nodeColor=<rgba 0 0.6719 1 1>}
 ]}
