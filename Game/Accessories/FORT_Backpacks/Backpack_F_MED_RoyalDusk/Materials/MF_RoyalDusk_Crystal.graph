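# Material function for the Royal Dusk backpack crystal, faking a refractive interior and blending new base color and emissive over input material attributes by Alpha.
# Depth-driven BumpOffset parallax shifts the interior texture, which is tinted, partly desaturated and lerped toward a fresnel color.
# It also builds sky-light reflection, backlight from atmosphere light direction and a normal map, and switchable sparks from two squared, scrolling noise layers.
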MaterialFunction /Game/Accessories/FORT_Backpacks/Backpack_F_MED_RoyalDusk/Materials/MF_RoyalDusk_Crystal
graph {nodes [
  n1 = NamedRerouteUsage()
  n2 = NamedRerouteUsage()
  n3 = NamedRerouteUsage()
  n4 = NamedRerouteUsage()
  n5 = VectorParameter() {group=RoyalDuskCrystal, parameterName=RDC_BaseColorMultiply, sortPriority=0, defaultValue=<rgba 1 1 1 1>}
  n6 = Multiply(A=@n5.RGB, B=NamedRerouteUsage())
  n7 = TextureSampleParameter2D(UVs=NamedRerouteUsage()) {
   group=RoyalDuskCrystal
   parameterName=RDC_InteriorTexture
   sortPriority=5
   texture=/Game/Accessories/FORT_Backpacks/Backpack_F_MED_RoyalDusk/Textures/T_RoyalDusk_RoyalDusk_Crystal1
  }
  n8 = OneMinus(None=ScalarParameter() {defaultValue=0.5, group=RoyalDuskCrystal, parameterName=RDC_InteriorPrismColors, sortPriority=2})
  n9 = Multiply(
   A=Multiply(A=@n6, B=Desaturation(None=@n7.RGB, Fraction=@n8))
   B=ScalarParameter() {defaultValue=12, group=RoyalDuskCrystal, parameterName=RDC_InteriorIntensity, sortPriority=1}
  ) {constB=0.1}
  n10 = ScalarParameter() {defaultValue=0.6, group=RoyalDuskCrystal, parameterName=RDC_InteriorContrast, sortPriority=3}
  n11 = VectorParameter() {group=RoyalDuskCrystal, parameterName=RDC_FresnelColor, sortPriority=19, defaultValue=<rgba 0.7823 0.3333 1 1>}
  n12 = LinearInterpolate(
   A=NamedRerouteDeclaration(None=LinearInterpolate(A=@n6, B=@n9, Alpha=@n10)) {name=InteriorColor, nodeColor=<rgba 1 0 0.4687 1>}
   B=@n11.RGB
   Alpha=NamedRerouteUsage()
  )
  n13 = Add(A=NamedRerouteUsage(), B=NamedRerouteUsage())
  n14 = NamedRerouteUsage()
  n15 = Add(
   A=StaticSwitchParameter(True=Add(A=@n13, B=@n14), False=@n13) {group=RoyalDuskCrystal, parameterName=RDC_Sparks, sortPriority=23}
   B=NamedRerouteUsage()
  )
  n16 = FunctionInput() {bUsePreviewValueAsDefault=true, inputName=Alpha, inputType=FunctionInput_Scalar, sortPriority=1, previewValue=<1 1 1 1>}
  FunctionOutput(None=BlendMaterialAttributes(
    A=SetMaterialAttributes!(MaterialAttributes=@n1, BaseColor=@n2, EmissiveColor=@n3) {attributeSetTypes [#1, #2]}
    B=SetMaterialAttributes(MaterialAttributes=@n4, BaseColor=@n12, EmissiveColor=@n15) {attributeSetTypes [#1, #2]}
    Alpha=@n16
   ))
  n17 = GetMaterialAttributes(None=FunctionInput() {inputName=Material, inputType=FunctionInput_MaterialAttributes}) {attributeGetTypes [#1, #2]}
  NamedRerouteDeclaration(None=@n17.EmissiveColor) {name=Emissive, nodeColor=<rgba 1 0 0.75 1>}
  NamedRerouteDeclaration(None=@n17.MaterialAttributes) {name=MA, nodeColor=<rgba 0.3406 1 0.0156 1>}
  n18 = TextureSampleParameter2D(UVs=NamedRerouteUsage()) {
   group=RoyalDuskCrystal
   parameterName=RDC_BacklightNormal
   samplerType=SAMPLERTYPE_Normal
   sortPriority=18
   texture=/Game/Accessories/FORT_Backpacks/Backpack_F_MED_RoyalDusk/Textures/T_RoyalDusk_RoyalDusk_CrystalNormal1
  }
  n19 = Transform(None=LinearInterpolate(
    A=Constant3Vector() {constant=<rgba 0 0 1 0>}
    B=@n18.RGB
    Alpha=ScalarParameter() {defaultValue=0.2, group=RoyalDuskCrystal, parameterName=RDC_BacklightNormalIntensity, sortPriority=15}
   ) {constAlpha=0.1})
  n20 = CameraVectorWS()
  n21 = Multiply(A=ScalarParameter() {defaultValue=1, group=RoyalDuskCrystal, parameterName=RDC_BacklightViewAngle, sortPriority=16}, B=@n20)
  n22 = DotProduct(A=Normalize(VectorInput=Add(A=@n19, B=@n21)), B=Multiply(A=SkyAtmosphereLightDirection()) {constB=-1})
  n23 = Saturate(None=Add(A=Multiply(A=@n22) {constB=0.5}) {constB=0.5})
  n24 = ScalarParameter() {defaultValue=2, group=RoyalDuskCrystal, parameterName=RDC_BacklightAperture, sortPriority=17}
  n25 = DotProduct(A=@n19, B=@n20)
  n26 = Multiply(A=Multiply(A=Power(Base=@n23, Exp=@n24), B=Saturate(None=@n25)), B=NamedRerouteUsage())
  n27 = VectorParameter() {group=RoyalDuskCrystal, parameterName=RDC_BacklightColorMultiply, sortPriority=13, defaultValue=<rgba 1 1 1 1>}
  NamedRerouteDeclaration(None=Multiply(
    A=Multiply(A=@n26, B=@n27.RGB)
    B=ScalarParameter() {defaultValue=2, group=RoyalDuskCrystal, parameterName=RDC_BackLightIntensity, sortPriority=14}
   )) {name=Backlight, nodeColor=<rgba 1 0 0.2578 1>}
  NamedRerouteDeclaration(None=@n17.BaseColor) {name=BaseColor, nodeColor=<rgba 0.1661 0.3253 1 1>}
  n28 = ReflectionVectorWS()
  n29 = ScalarParameter() {defaultValue=0.1, group=RoyalDuskCrystal, parameterName=RDC_ReflectionRoughness, sortPriority=22}
  n30 = VectorParameter() {group=RoyalDuskCrystal, parameterName=RDC_ReflectionColor, sortPriority=21, defaultValue=<rgba 0.0952 0.0364 0.1146 1>}
  NamedRerouteDeclaration(None=Multiply(A=SkyLightEnvMapSample(Direction=@n28, Roughness=@n29), B=@n30.RGB)) {name=Reflection, nodeColor=<rgba 1 0.1406 0 1>}
  n31 = TextureCoordinate()
  n32 = NamedRerouteDeclaration(None=@n31) {name=UVCoord, nodeColor=<rgba 0.5937 1 0 1>}
  n33 = ScalarParameter() {defaultValue=0.4, group=RoyalDuskCrystal, parameterName=RDC_InteriorUVTiling, sortPriority=4}
  n34 = ScalarParameter() {group=RoyalDuskCrystal, parameterName=RDC_DepthMinimum, sortPriority=6}
  n35 = ScalarParameter() {defaultValue=-10, group=RoyalDuskCrystal, parameterName=RDC_DepthMaximum, sortPriority=7}
  n36 = TextureSampleParameter2D(UVs=Multiply(A=@n31, B=ScalarParameter() {defaultValue=2.6, group=RoyalDuskCrystal, parameterName=RDC_DepthUVTiling, sortPriority=8})) {
   group=RoyalDuskCrystal
   parameterName=RDC_DepthTexture
   samplerType=SAMPLERTYPE_LinearGrayscale
   sortPriority=9
   texture=/Game/Accessories/FORT_Backpacks/Backpack_F_MED_RoyalDusk/Textures/T_RoyalDusk_RoyalDusk_Crystal2
  }
  n37 = NamedRerouteDeclaration(None=@n36.R) {name=SparksDepth, nodeColor=<rgba 0.1156 1 0.3086 1>}
  NamedRerouteDeclaration(None=BumpOffset(Coordinate=Multiply(A=@n32, B=@n33), Height=LinearInterpolate(A=@n34, B=@n35, Alpha=@n37) {constA=-4, constB=-20})) {name=UVOffsetCoord, nodeColor=<rgba 0.9453 0 1 1>}
  n38 = PixelNormalWS()
  n39 = CameraVectorWS()
  n40 = OneMinus(None=Saturate(None=DotProduct(A=@n38, B=@n39)))
  NamedRerouteDeclaration(None=Multiply(
    A=Multiply(A=@n40, B=@n40)
    B=ScalarParameter() {defaultValue=0.2, group=RoyalDuskCrystal, parameterName=RDC_FresnelIntensity, sortPriority=20}
   )) {name=Fresnel, nodeColor=<rgba 0 0.2734 1 1>}
  n41 = BumpOffset(
   Coordinate=NamedRerouteUsage()
   Height=Multiply(A=NamedRerouteUsage(), B=ScalarParameter() {defaultValue=-5, group=RoyalDuskCrystal, parameterName=RDC_SparksDepth, sortPriority=7})
  ) {heightRatio=0.01}
  n42 = ScalarParameter() {defaultValue=2, group=RoyalDuskCrystal, parameterName=RDC_SparksTiling, sortPriority=11}
  n43 = Multiply(A=@n42) {constB=1.2}
  n44 = Time()
  n45 = ScalarParameter() {defaultValue=0.01, group=RoyalDuskCrystal, parameterName=RDC_SparksSpeed1, sortPriority=11}
  n46 = TextureObjectParameter() {
   group=RoyalDuskCrystal
   parameterName=RDC_SparksTexture
   samplerType=SAMPLERTYPE_LinearGrayscale
   sortPriority=12
   texture=/Game/Accessories/FORT_Backpacks/Backpack_F_MED_RoyalDusk/Textures/T_RoyalDusk_RoyalDusk_CrystalSparks1
  }
  n47 = TextureSample(UVs=Add(A=Multiply(A=@n41, B=@n43), B=Multiply(A=@n44, B=@n45)), Tex=@n46) {
   samplerType=SAMPLERTYPE_LinearGrayscale
   texture=&T_RoyalDusk_RoyalDusk_CrystalSparks1
  }
  n48 = NamedRerouteUsage()
  n49 = ScalarParameter() {defaultValue=-0.005, group=RoyalDuskCrystal, parameterName=RDC_SparksSpeed2, sortPriority=11}
  n50 = TextureSample(UVs=Add(A=Multiply(A=@n42, B=@n48), B=Multiply(A=@n44, B=@n49)), Tex=@n46) {
   samplerType=SAMPLERTYPE_LinearGrayscale
   texture=&T_RoyalDusk_RoyalDusk_CrystalSparks1
  }
  n51 = NamedRerouteUsage()
  n52 = NamedRerouteUsage()
  n53 = Multiply(
   A=Multiply(A=Multiply(A=@n47.R, B=@n50.R), B=Add(A=@n51, B=@n52)) {constB=0}
   B=ScalarParameter() {defaultValue=14, group=RoyalDuskCrystal, parameterName=RDC_SparksIntensity, sortPriority=10}
  ) {constB=16}
  NamedRerouteDeclaration(None=Multiply(A=@n53, B=@n53)) {name=Sparks, nodeColor=<rgba 1 0.3611 0.0777 1>}
  Reroute(None=@n36.R)
 ]}
