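# M_RoyalDusk_Backpack is the skeletal-mesh material for the Royal Dusk backpack cosmetic, using material attributes and Default Lit shading.
# The shared 2023 character parent PreFX function feeds a backpack-specific function, whose attributes pass through a Royal Dusk crystal function.
# The result enters the parent PostFX function, which also receives PreFX outputs such as pre-skinned position, distance blend and Fresnel inputs.
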Material /Game/Accessories/FORT_Backpacks/Backpack_F_MED_RoyalDusk/Materials/M_RoyalDusk_Backpack
graph {
 properties {
  bUseMaterialAttributes=true
  bUsedWithMeshDeformer=true
  bUsedWithSkeletalMesh=true
  pixelDepthOffsetMode=PDOM_Legacy
  usedShadingModels=Default Lit
 }
 nodes [
  n1 = MaterialFunctionCall() {materialFunction=&MF_CharacterParent_2023_PreFX}
  MaterialFunctionCall(
   Material=MaterialFunctionCall!(
    Material=MaterialFunctionCall!(In=@n1.Result, LocalPosition=@n1.Result) {materialFunction=&/MF_RoyalDusk_Backpack}.MaterialAttributes
    Alpha
   ) {materialFunction=&/MF_RoyalDusk_Crystal}.Result
   M=@n1.M
   Pre Skinned Local Position=@n1.M
   Distance Blend=@n1.M
   Fresnel=@n1.M
  ) {materialFunction=&MF_CharacterParent_2023_PostFX}
 ]
}
